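# M_UI_TextMask is a translucent UI material that, going by its name, serves as a mask for text.
# A Texture parameter, defaulting to a grey-box white texture, has its alpha inverted through OneMinus, producing a cutout of the supplied shape.
# A constant value of one sits alongside it; the graph does not show which material outputs either value drives.
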Material /FortniteCoreUI/UI/Foundation/RockyRidge/Materials/M_UI_TextMask
graph {
 properties {
  bUsedWithStaticMesh=false
  blendMode=BLEND_Translucent
  materialDomain=MD_UI
  pixelDepthOffsetMode=PDOM_Legacy
  usedShadingModels=Default Lit
 }
 nodes [
  Constant() {r=1}
  OneMinus(None=TextureSampleParameter2D() {parameterName=Texture, texture=/Game/UI/GreyBox/white}.A)
 ]
}
