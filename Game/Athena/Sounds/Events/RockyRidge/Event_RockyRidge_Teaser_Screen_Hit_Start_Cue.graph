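# Sound cue for the RockyRidge event teaser, named as the start of a screen hit.
# A single wave player feeds a fixed short Delay node at the root, so the sound starts slightly after it is triggered.
# It uses the scripted events sound class and a very large max distance, with a concurrency override that allows one instance and stops the oldest.
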SoundCue /Game/Athena/Sounds/Events/RockyRidge/Event_RockyRidge_Teaser_Screen_Hit_Start_Cue
root=Delay_2
nodes [
 {id=Delay_2, children [WavePlayer_0], properties {delayMax=0.4, delayMin=0.4}}
 {id=WavePlayer_0, properties {soundWaveAssetPtr=&/Event_RockyRidge_Teaser_Screen_Hit_Start}}
]
properties {
 bOverrideConcurrency=true
 maxDistance=2097152
 soundClassObject=/Game/Sounds/FortSoundClassesAndModes/Fort_Events_Scripted
 volumeMultiplier=1
 concurrencyOverrides {maxCount=1, resolutionRule=StopOldest}
}
