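# Front-end RockyRidge event screen that plays a "To be Continued..." teaser, reveals the new season, then shows a countdown.
# PreConstruct splits ToBeContinuedText into lines; Construct or activation chains up to four line animations with hit sounds, skipping missing lines, then NewSeason, BGReveal and CountdownIntro.
# Tick displays time remaining until RockyRidgeStart from runtime options, clamped at zero.
# The close button quits Fortnite after a confirmation dialog.
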WidgetBlueprint /FortniteCoreUI/Foundation/RockyRidge/FrontEnd_RockyRidgeV2_Final_New
animations [
 ToBeContinued1 {
  binding_count=5
  length_seconds=0.45
  object=&self:ToBeContinued1
  sequence {bindings [
    Text_line4 {
     id=#1
     possessed_class=CommonTextBlock
     tracks [
      {
       type=MovieScene2DTransformTrack
       sections [{
         type=MovieScene2DTransformSection
         channels [
          Translation.X {type=MovieSceneScriptingFloatChannel}
          Translation.Y {type=MovieSceneScriptingFloatChannel}
          Angle {type=MovieSceneScriptingFloatChannel}
          Scale.X {type=MovieSceneScriptingFloatChannel, const [1,1]}
          Scale.Y {type=MovieSceneScriptingFloatChannel, const [1,1]}
          Shear.X {type=MovieSceneScriptingFloatChannel}
          Shear.Y {type=MovieSceneScriptingFloatChannel}
         ]
        }]
      }
      {
       type=MovieSceneFloatTrack
       display_name=Render Opacity
       sections [{type=MovieSceneFloatSection, channels [None {type=MovieSceneScriptingFloatChannel, const [0,1]}]}]
      }
     ]
    }
    Text_line2 {
     id=#2
     possessed_class=CommonTextBlock
     tracks [
      {
       type=MovieScene2DTransformTrack
       sections [{
         type=MovieScene2DTransformSection
         channels [
          Translation.X {type=MovieSceneScriptingFloatChannel, const [0,1]}
          Translation.Y {type=MovieSceneScriptingFloatChannel, const [0,1]}
          Angle {type=MovieSceneScriptingFloatChannel}
          Scale.X {type=MovieSceneScriptingFloatChannel, const [1,1]}
          Scale.Y {type=MovieSceneScriptingFloatChannel, const [1,1]}
          Shear.X {type=MovieSceneScriptingFloatChannel}
          Shear.Y {type=MovieSceneScriptingFloatChannel}
         ]
        }]
      }
      {
       type=MovieSceneFloatTrack
       display_name=Render Opacity
       sections [{type=MovieSceneFloatSection, channels [None {type=MovieSceneScriptingFloatChannel, const [0,1]}]}]
      }
     ]
    }
    CommonWidgetSwitcher_AnimStates {
     id=#3
     possessed_class=CommonWidgetSwitcherLegacy
     tracks [
      {
       type=MovieSceneFloatTrack
       display_name=Render Opacity
       sections [{type=MovieSceneFloatSection, channels [None {type=MovieSceneScriptingFloatChannel, keys [[0,0],[1.8973,1],[9,1]]}]}]
      }
      {
       type=MovieSceneIntegerTrack
       display_name=Active Widget Index
       sections [{type=MovieSceneIntegerSection, channels [None {type=MovieSceneScriptingIntegerChannel, const [1,1]}]}]
      }
     ]
    }
    Text_line3 {
     id=#4
     possessed_class=CommonTextBlock
     tracks [
      {
       type=MovieScene2DTransformTrack
       sections [{
         type=MovieScene2DTransformSection
         channels [
          Translation.X {type=MovieSceneScriptingFloatChannel}
          Translation.Y {type=MovieSceneScriptingFloatChannel}
          Angle {type=MovieSceneScriptingFloatChannel}
          Scale.X {type=MovieSceneScriptingFloatChannel, const [1,1]}
          Scale.Y {type=MovieSceneScriptingFloatChannel, const [1,1]}
          Shear.X {type=MovieSceneScriptingFloatChannel}
          Shear.Y {type=MovieSceneScriptingFloatChannel}
         ]
        }]
      }
      {
       type=MovieSceneFloatTrack
       display_name=Render Opacity
       sections [{type=MovieSceneFloatSection, channels [None {type=MovieSceneScriptingFloatChannel, const [0,1]}]}]
      }
     ]
    }
    Text_line1 {
     id=#5
     possessed_class=CommonTextBlock
     tracks [
      {
       type=MovieScene2DTransformTrack
       sections [{
         type=MovieScene2DTransformSection
         channels [
          Translation.X {type=MovieSceneScriptingFloatChannel, const [0,1]}
          Translation.Y {type=MovieSceneScriptingFloatChannel, const [0,1]}
          Angle {type=MovieSceneScriptingFloatChannel}
          Scale.X {type=MovieSceneScriptingFloatChannel, keys [[0,3],[5.6937,0.755],[6,1]]}
          Scale.Y {type=MovieSceneScriptingFloatChannel, keys [[0,3],[5.6937,0.755],[6,1]]}
          Shear.X {type=MovieSceneScriptingFloatChannel}
          Shear.Y {type=MovieSceneScriptingFloatChannel}
         ]
        }]
      }
      {
       type=MovieSceneFloatTrack
       display_name=Render Opacity
       sections [{type=MovieSceneFloatSection, channels [None {type=MovieSceneScriptingFloatChannel, keys [[0,0],[3.418,1]]}]}]
      }
     ]
    }
   ]}
 }
 ToBeContinued2 {
  binding_count=4
  length_seconds=0.45
  object=&self:ToBeContinued2
  sequence {bindings [
    Text_line4 {
     id=#1
     possessed_class=CommonTextBlock
     tracks [
      {
       type=MovieScene2DTransformTrack
       sections [{
         type=MovieScene2DTransformSection
         channels [
          Translation.X {type=MovieSceneScriptingFloatChannel}
          Translation.Y {type=MovieSceneScriptingFloatChannel}
          Angle {type=MovieSceneScriptingFloatChannel}
          Scale.X {type=MovieSceneScriptingFloatChannel, const [1,1]}
          Scale.Y {type=MovieSceneScriptingFloatChannel, const [1,1]}
          Shear.X {type=MovieSceneScriptingFloatChannel}
          Shear.Y {type=MovieSceneScriptingFloatChannel}
         ]
        }]
      }
      {
       type=MovieSceneFloatTrack
       display_name=Render Opacity
       sections [{type=MovieSceneFloatSection, channels [None {type=MovieSceneScriptingFloatChannel, const [0,1]}]}]
      }
     ]
    }
    Text_line2 {
     id=#2
     possessed_class=CommonTextBlock
     tracks [
      {
       type=MovieSceneFloatTrack
       display_name=Render Opacity
       sections [{type=MovieSceneFloatSection, channels [None {type=MovieSceneScriptingFloatChannel, keys [[0,0],[3.418,1]]}]}]
      }
      {
       type=MovieScene2DTransformTrack
       sections [{
         type=MovieScene2DTransformSection
         channels [
          Translation.X {type=MovieSceneScriptingFloatChannel, const [0,1]}
          Translation.Y {type=MovieSceneScriptingFloatChannel, const [0,1]}
          Angle {type=MovieSceneScriptingFloatChannel}
          Scale.X {type=MovieSceneScriptingFloatChannel, keys [[0,3],[5.827,0.9037],[6,1]]}
          Scale.Y {type=MovieSceneScriptingFloatChannel, keys [[0,3],[5.827,0.9037],[6,1]]}
          Shear.X {type=MovieSceneScriptingFloatChannel}
          Shear.Y {type=MovieSceneScriptingFloatChannel}
         ]
        }]
      }
     ]
    }
    Text_line3 {
     id=#4
     possessed_class=CommonTextBlock
     tracks [
      {
       type=MovieScene2DTransformTrack
       sections [{
         type=MovieScene2DTransformSection
         channels [
          Translation.X {type=MovieSceneScriptingFloatChannel}
          Translation.Y {type=MovieSceneScriptingFloatChannel}
          Angle {type=MovieSceneScriptingFloatChannel}
          Scale.X {type=MovieSceneScriptingFloatChannel, const [1,1]}
          Scale.Y {type=MovieSceneScriptingFloatChannel, const [1,1]}
          Shear.X {type=MovieSceneScriptingFloatChannel}
          Shear.Y {type=MovieSceneScriptingFloatChannel}
         ]
        }]
      }
      {
       type=MovieSceneFloatTrack
       display_name=Render Opacity
       sections [{type=MovieSceneFloatSection, channels [None {type=MovieSceneScriptingFloatChannel, const [0,1]}]}]
      }
     ]
    }
    Text_line1 {
     id=#5
     possessed_class=CommonTextBlock
     tracks [
      {
       type=MovieScene2DTransformTrack
       sections [{
         type=MovieScene2DTransformSection
         channels [
          Translation.X {type=MovieSceneScriptingFloatChannel, keys [[0,0],[3.91,-251.6722]]}
          Translation.Y {type=MovieSceneScriptingFloatChannel, keys [[0,0],[3.91,-227.0869]]}
          Angle {type=MovieSceneScriptingFloatChannel}
          Scale.X {type=MovieSceneScriptingFloatChannel, keys [[0,1],[3.91,0.5]]}
          Scale.Y {type=MovieSceneScriptingFloatChannel, keys [[0,1],[3.91,0.5]]}
          Shear.X {type=MovieSceneScriptingFloatChannel}
          Shear.Y {type=MovieSceneScriptingFloatChannel}
         ]
        }]
      }
      {
       type=MovieSceneFloatTrack
       display_name=Render Opacity
       sections [{type=MovieSceneFloatSection, channels [None {type=MovieSceneScriptingFloatChannel, const [1,2]}]}]
      }
     ]
    }
   ]}
 }
 ToBeContinued3 {
  binding_count=4
  length_seconds=0.45
  object=&self:ToBeContinued3
  sequence {bindings [
    Text_line4 {
     id=#1
     possessed_class=CommonTextBlock
     tracks [
      {
       type=MovieScene2DTransformTrack
       sections [{
         type=MovieScene2DTransformSection
         channels [
          Translation.X {type=MovieSceneScriptingFloatChannel}
          Translation.Y {type=MovieSceneScriptingFloatChannel}
          Angle {type=MovieSceneScriptingFloatChannel}
          Scale.X {type=MovieSceneScriptingFloatChannel, const [1,1]}
          Scale.Y {type=MovieSceneScriptingFloatChannel, const [1,1]}
          Shear.X {type=MovieSceneScriptingFloatChannel}
          Shear.Y {type=MovieSceneScriptingFloatChannel}
         ]
        }]
      }
      {
       type=MovieSceneFloatTrack
       display_name=Render Opacity
       sections [{type=MovieSceneFloatSection, channels [None {type=MovieSceneScriptingFloatChannel, const [0,1]}]}]
      }
     ]
    }
    Text_line2 {
     id=#2
     possessed_class=CommonTextBlock
     tracks [
      {
       type=MovieSceneFloatTrack
       display_name=Render Opacity
       sections [{type=MovieSceneFloatSection, channels [None {type=MovieSceneScriptingFloatChannel, const [1,1]}]}]
      }
      {
       type=MovieScene2DTransformTrack
       sections [{
         type=MovieScene2DTransformSection
         channels [
          Translation.X {type=MovieSceneScriptingFloatChannel, keys [[0,0],[6,-18.5877]]}
          Translation.Y {type=MovieSceneScriptingFloatChannel, keys [[0,0],[6,-150.5964]]}
          Angle {type=MovieSceneScriptingFloatChannel}
          Scale.X {type=MovieSceneScriptingFloatChannel, keys [[0,1],[6,0.5]]}
          Scale.Y {type=MovieSceneScriptingFloatChannel, keys [[0,1],[6,0.5]]}
          Shear.X {type=MovieSceneScriptingFloatChannel}
          Shear.Y {type=MovieSceneScriptingFloatChannel}
         ]
        }]
      }
     ]
    }
    Text_line3 {
     id=#4
     possessed_class=CommonTextBlock
     tracks [
      {
       type=MovieSceneFloatTrack
       display_name=Render Opacity
       sections [{type=MovieSceneFloatSection, channels [None {type=MovieSceneScriptingFloatChannel, keys [[0,0],[3.418,1]]}]}]
      }
      {
       type=MovieScene2DTransformTrack
       sections [{
         type=MovieScene2DTransformSection
         channels [
          Translation.X {type=MovieSceneScriptingFloatChannel, const [0,1]}
          Translation.Y {type=MovieSceneScriptingFloatChannel, const [0,1]}
          Angle {type=MovieSceneScriptingFloatChannel}
          Scale.X {type=MovieSceneScriptingFloatChannel, keys [[0,3],[5.827,0.9037],[6,1]]}
          Scale.Y {type=MovieSceneScriptingFloatChannel, keys [[0,3],[5.827,0.9037],[6,1]]}
          Shear.X {type=MovieSceneScriptingFloatChannel}
          Shear.Y {type=MovieSceneScriptingFloatChannel}
         ]
        }]
      }
     ]
    }
    Text_line1 {
     id=#5
     possessed_class=CommonTextBlock
     tracks [
      {
       type=MovieScene2DTransformTrack
       sections [{
         type=MovieScene2DTransformSection
         channels [
          Translation.X {type=MovieSceneScriptingFloatChannel, keys [[0,-251.6722],[6,-162.2781]]}
          Translation.Y {type=MovieSceneScriptingFloatChannel, keys [[0,-227.0869],[6,-355.9444]]}
          Angle {type=MovieSceneScriptingFloatChannel}
          Scale.X {type=MovieSceneScriptingFloatChannel, keys [[0,0.5],[6,0.25]]}
          Scale.Y {type=MovieSceneScriptingFloatChannel, keys [[0,0.5],[6,0.25]]}
          Shear.X {type=MovieSceneScriptingFloatChannel}
          Shear.Y {type=MovieSceneScriptingFloatChannel}
         ]
        }]
      }
      {
       type=MovieSceneFloatTrack
       display_name=Render Opacity
       sections [{type=MovieSceneFloatSection, channels [None {type=MovieSceneScriptingFloatChannel, const [1,2]}]}]
      }
     ]
    }
   ]}
 }
 ToBeContinued4 {
  binding_count=4
  length_seconds=0.45
  object=&self:ToBeContinued4
  sequence {bindings [
    Text_line4 {
     id=#1
     possessed_class=CommonTextBlock
     tracks [
      {
       type=MovieSceneFloatTrack
       display_name=Render Opacity
       sections [{type=MovieSceneFloatSection, channels [None {type=MovieSceneScriptingFloatChannel, keys [[0,0],[3.418,1]]}]}]
      }
      {
       type=MovieScene2DTransformTrack
       sections [{
         type=MovieScene2DTransformSection
         channels [
          Translation.X {type=MovieSceneScriptingFloatChannel}
          Translation.Y {type=MovieSceneScriptingFloatChannel}
          Angle {type=MovieSceneScriptingFloatChannel}
          Scale.X {type=MovieSceneScriptingFloatChannel, keys [[0,3],[5.827,0.9037],[6,1]]}
          Scale.Y {type=MovieSceneScriptingFloatChannel, keys [[0,3],[5.827,0.9037],[6,1]]}
          Shear.X {type=MovieSceneScriptingFloatChannel}
          Shear.Y {type=MovieSceneScriptingFloatChannel}
         ]
        }]
      }
     ]
    }
    Text_line2 {
     id=#2
     possessed_class=CommonTextBlock
     tracks [
      {
       type=MovieSceneFloatTrack
       display_name=Render Opacity
       sections [{type=MovieSceneFloatSection, channels [None {type=MovieSceneScriptingFloatChannel, const [1,1]}]}]
      }
      {
       type=MovieScene2DTransformTrack
       sections [{
         type=MovieScene2DTransformSection
         channels [
          Translation.X {type=MovieSceneScriptingFloatChannel, keys [[0,-18.5877],[6,-51.909]]}
          Translation.Y {type=MovieSceneScriptingFloatChannel, keys [[0,-150.5964],[6,-241.9127]]}
          Angle {type=MovieSceneScriptingFloatChannel}
          Scale.X {type=MovieSceneScriptingFloatChannel, keys [[0,0.5],[6,0.25]]}
          Scale.Y {type=MovieSceneScriptingFloatChannel, keys [[0,0.5],[6,0.25]]}
          Shear.X {type=MovieSceneScriptingFloatChannel}
          Shear.Y {type=MovieSceneScriptingFloatChannel}
         ]
        }]
      }
     ]
    }
    Text_line3 {
     id=#4
     possessed_class=CommonTextBlock
     tracks [
      {
       type=MovieScene2DTransformTrack
       sections [{
         type=MovieScene2DTransformSection
         channels [
          Translation.X {type=MovieSceneScriptingFloatChannel, keys [[0,0],[6,-18.5877]]}
          Translation.Y {type=MovieSceneScriptingFloatChannel, keys [[0,0],[6,-150.5964]]}
          Angle {type=MovieSceneScriptingFloatChannel}
          Scale.X {type=MovieSceneScriptingFloatChannel, keys [[0,1],[6,0.5]]}
          Scale.Y {type=MovieSceneScriptingFloatChannel, keys [[0,1],[6,0.5]]}
          Shear.X {type=MovieSceneScriptingFloatChannel}
          Shear.Y {type=MovieSceneScriptingFloatChannel}
         ]
        }]
      }
      {
       type=MovieSceneFloatTrack
       display_name=Render Opacity
       sections [{type=MovieSceneFloatSection, channels [None {type=MovieSceneScriptingFloatChannel, const [1,1]}]}]
      }
     ]
    }
    Text_line1 {
     id=#5
     possessed_class=CommonTextBlock
     tracks [
      {
       type=MovieScene2DTransformTrack
       sections [{
         type=MovieScene2DTransformSection
         channels [
          Translation.X {type=MovieSceneScriptingFloatChannel, keys [[0,-162.2781],[6,-102.7704]]}
          Translation.Y {type=MovieSceneScriptingFloatChannel, keys [[0,-355.9444],[6,-424.3133]]}
          Angle {type=MovieSceneScriptingFloatChannel}
          Scale.X {type=MovieSceneScriptingFloatChannel, keys [[0,0.25],[6,0.2]]}
          Scale.Y {type=MovieSceneScriptingFloatChannel, keys [[0,0.25],[6,0.2]]}
          Shear.X {type=MovieSceneScriptingFloatChannel}
          Shear.Y {type=MovieSceneScriptingFloatChannel}
         ]
        }]
      }
      {
       type=MovieSceneFloatTrack
       display_name=Render Opacity
       sections [{type=MovieSceneFloatSection, channels [None {type=MovieSceneScriptingFloatChannel, const [1,2]}]}]
      }
     ]
    }
   ]}
 }
 NewSeason {
  binding_count=4
  length_seconds=2.8798
  object=&self:NewSeason
  sequence {bindings [
    Overlay_1 {
     id=#6
     possessed_class=Overlay
     tracks [{
       type=MovieSceneFloatTrack
       display_name=Render Opacity
       sections [{type=MovieSceneFloatSection, channels [None {type=MovieSceneScriptingFloatChannel, keys [[0,0],[13.7343,1],[53.5823,1],[57.595,0]]}]}]
      }]
    }
    RetainerBox_NewSeason {
     id=#7
     display_name=RetainerBox_0
     possessed_class=RetainerBox
     tracks [{
       type=MovieSceneFloatTrack
       display_name=Render Opacity
       sections [{type=MovieSceneFloatSection, channels [None {type=MovieSceneScriptingFloatChannel, keys [[0,0],[13.7343,1],[53.5823,1],[57.595,0]]}]}]
      }]
    }
    VerticalBox {
     id=#8
     possessed_class=VerticalBox
     tracks [{
       type=MovieScene2DTransformTrack
       sections [{
         type=MovieScene2DTransformSection
         channels [
          Translation.X {type=MovieSceneScriptingFloatChannel}
          Translation.Y {type=MovieSceneScriptingFloatChannel}
          Angle {type=MovieSceneScriptingFloatChannel}
          Scale.X {type=MovieSceneScriptingFloatChannel, keys [[0,15],[15.864,1],[16.342,0.7881],[19,1],[53.5823,1],[57.595,0.5]]}
          Scale.Y {type=MovieSceneScriptingFloatChannel, keys [[0,15],[15.864,1],[16.342,0.7881],[19,1],[53.5823,1],[57.595,0.5]]}
          Shear.X {type=MovieSceneScriptingFloatChannel}
          Shear.Y {type=MovieSceneScriptingFloatChannel}
         ]
        }]
      }]
    }
    CommonWidgetSwitcher_AnimStates {
     id=#9
     possessed_class=CommonWidgetSwitcherLegacy
     tracks [{
       type=MovieSceneIntegerTrack
       display_name=Active Widget Index
       sections [{type=MovieSceneIntegerSection, channels [None {type=MovieSceneScriptingIntegerChannel, const [2,1]}]}]
      }]
    }
   ]}
 }
 CountdownIntro {
  binding_count=6
  length_seconds=1.7
  object=&self:CountdownIntro
  sequence {bindings [
    CloseButton {
     id=#10
     possessed_class=SkewButton_C
     tracks [{
       type=MovieSceneFloatTrack
       display_name=Render Opacity
       sections [{type=MovieSceneFloatSection, channels [None {type=MovieSceneScriptingFloatChannel, keys [[26.4637,0],[34,1]]}]}]
      }]
    }
    VerticalBox_Countdown {
     id=#11
     possessed_class=VerticalBox
     tracks [
      {
       type=MovieScene2DTransformTrack
       sections [{
         type=MovieScene2DTransformSection
         channels [
          Translation.X {type=MovieSceneScriptingFloatChannel}
          Translation.Y {type=MovieSceneScriptingFloatChannel}
          Angle {type=MovieSceneScriptingFloatChannel}
          Scale.X {type=MovieSceneScriptingFloatChannel, keys [[6.4557,0.5],[13.072,1]]}
          Scale.Y {type=MovieSceneScriptingFloatChannel, keys [[6.4557,0.5],[13.072,1]]}
          Shear.X {type=MovieSceneScriptingFloatChannel}
          Shear.Y {type=MovieSceneScriptingFloatChannel}
         ]
        }]
      }
      {
       type=MovieSceneFloatTrack
       display_name=Render Opacity
       sections [{type=MovieSceneFloatSection, channels [None {type=MovieSceneScriptingFloatChannel, keys [[6.4557,0],[13.072,1]]}]}]
      }
     ]
    }
    CommonWidgetSwitcher_AnimStates {
     id=#12
     possessed_class=CommonWidgetSwitcherLegacy
     tracks [{
       type=MovieSceneIntegerTrack
       display_name=Active Widget Index
       sections [{type=MovieSceneIntegerSection, channels [None {type=MovieSceneScriptingIntegerChannel, const [3,1]}]}]
      }]
    }
    Text_Season5_Line1 {
     id=#13
     display_name=Image_Logo
     possessed_class=CommonTextBlock
     tracks [{
       type=MovieScene2DTransformTrack
       sections [{
         type=MovieScene2DTransformSection
         channels [
          Translation.X {type=MovieSceneScriptingFloatChannel}
          Translation.Y {type=MovieSceneScriptingFloatChannel}
          Angle {type=MovieSceneScriptingFloatChannel}
          Scale.X {type=MovieSceneScriptingFloatChannel, keys [[0,0],[21,1.0042],[24,1]]}
          Scale.Y {type=MovieSceneScriptingFloatChannel, keys [[0,0],[21,1.0042],[24,1]]}
          Shear.X {type=MovieSceneScriptingFloatChannel}
          Shear.Y {type=MovieSceneScriptingFloatChannel}
         ]
        }]
      }]
    }
    Overlay_CountdownSlide {
     id=#14
     possessed_class=Overlay
     tracks [{
       type=MovieSceneFloatTrack
       display_name=Render Opacity
       sections [{type=MovieSceneFloatSection, channels [None {type=MovieSceneScriptingFloatChannel, const [1,2]}]}]
      }]
    }
    Text_Season5_Line2 {
     id=#15
     possessed_class=CommonTextBlock
     tracks [{
       type=MovieScene2DTransformTrack
       sections [{
         type=MovieScene2DTransformSection
         channels [
          Translation.X {type=MovieSceneScriptingFloatChannel}
          Translation.Y {type=MovieSceneScriptingFloatChannel}
          Angle {type=MovieSceneScriptingFloatChannel}
          Scale.X {type=MovieSceneScriptingFloatChannel, keys [[0,0],[22.256,1],[24,1]]}
          Scale.Y {type=MovieSceneScriptingFloatChannel, keys [[0,0],[22.256,1],[24,1]]}
          Shear.X {type=MovieSceneScriptingFloatChannel}
          Shear.Y {type=MovieSceneScriptingFloatChannel}
         ]
        }]
      }]
    }
   ]}
 }
 FadeIn {
  binding_count=2
  length_seconds=3.038
  object=&self:FadeIn
  sequence {bindings [
    CommonWidgetSwitcher_AnimStates {
     id=#16
     possessed_class=CommonWidgetSwitcherLegacy
     tracks [{
       type=MovieSceneIntegerTrack
       display_name=Active Widget Index
       sections [{type=MovieSceneIntegerSection, channels [None {type=MovieSceneScriptingIntegerChannel, const [0,1]}]}]
      }]
    }
    Image_Fade {
     id=#17
     possessed_class=Image
     tracks [{
       type=MovieSceneFloatTrack
       display_name=Render Opacity
       sections [{type=MovieSceneFloatSection, channels [None {type=MovieSceneScriptingFloatChannel, keys [[0,1],[60.7593,0]]}]}]
      }]
    }
   ]}
 }
 BGReveal {
  binding_count=1
  length_seconds=0.25
  object=&self:BGReveal
  sequence {bindings [Overlay_Background {
     id=#18
     possessed_class=Overlay
     tracks [{
       type=MovieSceneFloatTrack
       display_name=Render Opacity
       sections [{type=MovieSceneFloatSection, end_frame=5, start_frame=0, channels [None {type=MovieSceneScriptingFloatChannel, keys [[0,0],[5,1]]}]}]
      }]
    }]}
 }
]
graphs [
 EventGraph {nodes [
   "On Button Hovered (CloseButton)"!() -> (OutputDelegate, then, Button:Common Button Legacy Object Reference) {is=title K2Node_ComponentBoundEvent}
   > SetColorAndOpacity(
    self=Get Text_Button_Close() -> (Text_Button_Close:Common Text Object Reference)
    InColorAndOpacity=Make Slate Color(SpecifiedColor=(R=1,G=1,B=1,A=1))
   )
   "On Button Clicked (CloseButton)"!() -> (OutputDelegate, then, Button:Common Button Legacy Object Reference) {is=title K2Node_ComponentBoundEvent}
   > n1 = ShowSimpleConfirmationDialog_CustomInput!(
    execute:Exec
    Icon:Slate Brush Structure
    Title:Text="LOCTEXT(\"Exit\")"
    Message:Text="LOCTEXT(\"Are you sure you want to exit Fortnite?\")"
    ConfirmAction:Name=Confirm
    DeclineAction:Name=Cancel
    AdditionalContent:Widget Object Reference
    LeftAdditionalContent:Widget Object Reference
    Cancelable:bool=true
    TargetPlayer:Local Player Object Reference
   ) -> (then:Exec, DialogResult:Exec, Result:EFortDialogResult Enum, ResultName:Name) {is=title K2Node_AsyncAction}
   Switch on EFortDialogResult!(execute=@n1.DialogResult, Selection:EFortDialogResult Enum=@n1) -> (Confirmed:Exec, Declined:Exec, Ignored:Exec, Killed:Exec, TimedOut:Exec) {is=title K2Node_SwitchEnum}
   > QuitGame(self:Fort Global UIContext Object Reference=Get FortGlobalUIContext!() -> (ReturnValue:Fort Global UIContext Object Reference) {is=title K2Node_GetBlueprintContextEx})
   "On Button Unhovered (CloseButton)"!() -> (OutputDelegate, then, Button:Common Button Legacy Object Reference) {is=title K2Node_ComponentBoundEvent}
   > SetColorAndOpacity(self=Get Text_Button_Close(), InColorAndOpacity=Make Slate Color(SpecifiedColor=(R=0,G=0,B=0,A=1)))
   Event PreConstruct()
   n2 = Get ToBeContinuedText() -> (ToBeContinuedText:Text)
   > Set TextArray(TextArray:Array of Strings=ParseIntoArray(SourceString="To String (Text)"(InText=@n2), Delimiter="\r\n")) -> (Output_Get:Array of Strings)
   > n3 = For Each Loop(Array:Array of Widget Object References=GetAllChildren(self=Get Overlay_TextLines() -> (Overlay_TextLines:Overlay Object Reference))) -> (Array Element:Widget Object Reference)
   > n4 = Cast To CommonTextBlock!(execute, Object=@n3) -> (then, CastFailed, AsCommon Text:Common Text Object Reference) {is=title K2Node_DynamicCast}
   n5 = Get TextArray() -> (TextArray:Array of Strings)
   > "SetText (Text)"(self=@n4, InText="To Text (String)"(InString="Get (a copy)"(Array=@n5, Dimension 1=@n3.Array Index)))
   n6 = Event Construct()
   PlaySound2D(execute=[@n6.then, @n18.then], Sound=/Game/Athena/Sounds/Events/RockyRidge/Event_RockyRidge_Teaser_Screen_Hit_Start_Cue)
   > PushSoundMixModifier!(execute, InSoundMixModifier:Sound Mix Object Reference=/Game/Athena/Sounds/Events/RockyRidge/RockyRidge_Mix) -> (then)
   > n7 = "Play Animation with Finished event (legacy)"!(
    execute:Exec
    Widget:User Widget Object Reference=Self-Reference()
    InAnimation:Widget Animation Object Reference=Get FadeIn() -> (FadeIn:Widget Animation Object Reference)
    StartAtTime:float=0
    NumLoopsToPlay:int=1
    PlayMode:EUMGSequencePlayMode Enum
    PlaybackSpeed:float=1
   ) -> (then:Exec, Result:UMGSequence Player Object Reference, Finished:Exec) {is=title K2Node_PlayAnimation}
   n8 = "Play Animation with Finished event (legacy)"(
    execute=@n7.Finished
    Widget=Self-Reference()
    InAnimation=Get ToBeContinued1() -> (ToBeContinued1:Widget Animation Object Reference)
    StartAtTime=0
    NumLoopsToPlay=1
    PlaybackSpeed=2
   )
   Delay(execute=@n8.Finished, Duration=0.5)
   n9 = Length(TargetArray:Array of Strings=Get TextArray())
   > n10 = Branch(Condition="integer > integer"(A=@n9, B=1) {is=title K2Node_CallFunction})
   > SoundMaybe!(
    execute
    self
    Index:int=1
    Sound:Sound Base Object Reference=/Game/Athena/Sounds/Events/RockyRidge/Event_RockyRidge_Teaser_Screen_Hit_Continue
   ) -> (then)
   > n11 = "Play Animation with Finished event (legacy)"(
    Widget=Self-Reference()
    InAnimation=Get ToBeContinued2() -> (ToBeContinued2:Widget Animation Object Reference)
    StartAtTime=0
    NumLoopsToPlay=1
    PlaybackSpeed=2
   )
   Delay(execute=@n11.Finished, Duration=0.5)
   > n12 = Branch(Condition="integer > integer"(A=@n9, B=2))
   > SoundMaybe(Index=2, Sound=/Game/Athena/Sounds/Events/RockyRidge/Event_RockyRidge_Teaser_Screen_Hit_Tomorrow)
   > n13 = "Play Animation with Finished event (legacy)"(
    Widget=Self-Reference()
    InAnimation=Get ToBeContinued3() -> (ToBeContinued3:Widget Animation Object Reference)
    StartAtTime=0
    NumLoopsToPlay=1
    PlaybackSpeed=2
   )
   Delay(execute=@n13.Finished, Duration=0.5)
   > n14 = Branch(Condition="integer > integer"(A=@n9, B=3))
   > SoundMaybe(Index=3, Sound=&Event_RockyRidge_Teaser_Screen_Hit_Tomorrow)
   > n15 = "Play Animation with Finished event (legacy)"(
    Widget=Self-Reference()
    InAnimation=Get ToBeContinued4() -> (ToBeContinued4:Widget Animation Object Reference)
    StartAtTime=0
    NumLoopsToPlay=1
    PlaybackSpeed=2
   )
   Delay(execute=[@n15.Finished, @n14.else, @n12.else, @n10.else], Duration=2)
   > PlaySound2D(Sound=/Game/Athena/Sounds/Events/RockyRidge/Event_RockyRidge_Teaser_Screen_Hit_S5)
   > n16 = "Play Animation with Finished event (legacy)"(
    Widget=Self-Reference()
    InAnimation=Get NewSeason() -> (NewSeason:Widget Animation Object Reference)
    StartAtTime=0
    NumLoopsToPlay=1
    PlaybackSpeed=1
   )
   > Delay(Duration=0.4)
   > PlayAnimation(InAnimation=Get BGReveal() -> (BGReveal:Widget Animation Object Reference))
   Delay(execute=@n16.Finished, Duration=0)
   > PlaySound2D(Sound=/Game/Athena/Sounds/Events/RockyRidge/Event_RockyRidge_Teaser_Screen_Hit_End)
   > n17 = "Play Animation with Finished event (legacy)"(
    Widget=Self-Reference()
    InAnimation=Get CountdownIntro() -> (CountdownIntro:Widget Animation Object Reference)
    StartAtTime=0
    NumLoopsToPlay=1
    PlaybackSpeed=1
   )
   Reroute Node(InputPin=@n17.Finished)
   Event Tick()
   > UpdateCountdown()
   n18 = Event On Activated()
  ]}
 SoundMaybe {nodes [
   n1 = SoundMaybe!() -> (then, Index:int, Sound:Sound Base Object Reference) {is=title K2Node_FunctionEntry}
   n2 = Last Index!(TargetArray=Get TextArray()) -> (ReturnValue=0)
   > Branch(Condition=OR Boolean(A="Equal (Integer)"(A=@n1, B=@n2) {is=title K2Node_CallFunction}, B="Equal (Integer)"(A=@n1, B=1)))
   > PlaySound2D(Sound=@n1.Sound)
  ]}
 UpdateCountdown {nodes [
   UpdateCountdown() {is=title K2Node_FunctionEntry}
   n1 = DateTime - DateTime!(
    A:Date Time Structure=Get RockyRidgeStart!(self:Fort Runtime Options Object Reference=GetRuntimeOptions()) -> (RockyRidgeStart:Date Time Structure)
    B:Date Time Structure=UTC Now!() -> (ReturnValue:Date Time Structure)
   ) -> (ReturnValue:Timespan Structure)
   n2 = "Timespan >= Timespan"!(
    A:Timespan Structure=@n1
    B:Timespan Structure=MakeTimespan!(Days:int=0, Hours:int=0, Minutes:int=0, Seconds:int=0, Milliseconds:int=0) -> (ReturnValue:Timespan Structure)
   ) -> (ReturnValue:bool)
   n3 = MakeTimespan(Days=0, Hours=0, Minutes=0, Seconds=0, Milliseconds=0)
   > "SetText (Text)"(
    self=Get Text_Countdown_value() -> (Text_Countdown_value:Common Text Object Reference)
    InText=GetTimespanAsText!(
     Timespan:"Timespan Structure (by ref)"=Select(Option 0:"Timespan Structure (by ref)"=@n3, Option 1:"Timespan Structure (by ref)"=@n1, Index=@n2) -> (ReturnValue:"Timespan Structure (by ref)")
     bShowHoursWhenZero:bool=true
     bShowSeconds:bool=true
     FormatType:ETimespanAsTextFormat Enum=Colons
     UnitDisplayType:ETimespanUnitDisplayFormat Enum=Full
     bUseSentenceCase:bool=false
    ) -> (ReturnValue:Text)
   )
  ]}
]
variables [
 CurrentTime {type=number}
 TextArray {type="array<string>", default []}
 ToBeContinuedText {type=string, default="To be\r\nContinued..."}
]
widgets {
 info {
  inheritedWidgetCount=4
  namedSlotCount=1
  parentClass=/Script/FortniteUI.FortUIStateWidget_NightNightBase
  rootWidgetClass=&Overlay
  widgetCount=54
 }
 widgets [
  Overlay_0 {widgetClassPath=&Overlay, widget=Overlay_0, properties {clipping=ClipToBounds}}
  Image_BGColor {
   widgetClassPath=&Image
   bIsVariable=true
   parent=Overlay_0
   widget=Image_BGColor
   properties {
    brush {
     bIsDynamicallyLoaded=false
     drawAs=Image
     imageType=NoImage
     mirroring=NoMirror
     resourceName=None
     resourceObject=/Game/Characters/Player/Male/Medium/Bodies/M_MED_HighTower_Tapas/Textures/T_Hightower_Tapas_Star
     tiling=NoTile
     imageSize=<2574.041 1024>
     margin=<ltrb 0 0 0 0>
     outlineSettings {
      bUseBrushTransparency=false
      roundingType=HalfHeightRadius
      width=0
      color {colorUseRule=UseColor_Specified, specifiedColor=<rgba 0 0 0 0>}
      cornerRadii=<0 0 0 0>
     }
     tintColor {colorUseRule=UseColor_Specified, specifiedColor=<rgba 1 1 1 1>}
     uVRegion {bIsValid=false, max=<0 0>, min=<0 0>}
    }
    colorAndOpacity=<rgba 0.2977 0.3244 1 1>
   }
   slot {class=OverlaySlot, properties {horizontalAlignment=HAlign_Fill, verticalAlignment=VAlign_Fill}}
  }
  ScaleBox_1 {
   widgetClassPath=&ScaleBox
   parent=Overlay_0
   widget=ScaleBox_1
   properties {stretch=None}
   slot {class=OverlaySlot, properties {horizontalAlignment=HAlign_Fill, verticalAlignment=VAlign_Fill}}
  }
  Overlay_Background {
   widgetClassPath=&Overlay
   parent=ScaleBox_1
   widget=Overlay_Background
   properties {renderOpacity=0}
   slot {class=ScaleBoxSlot, properties {horizontalAlignment=HAlign_Fill, verticalAlignment=VAlign_Fill}}
  }
  Glow_0 {
   widgetClassPath=&Image
   bIsVariable=true
   parent=Overlay_Background
   widget=Glow_0
   properties {
    visibility=SelfHitTestInvisible
    brush {
     bIsDynamicallyLoaded=false
     drawAs=Image
     imageType=NoImage
     mirroring=NoMirror
     resourceName=None
     resourceObject=/Game/UI/Frontend/HeroLoadout/Art/Materials/FX/M_UI_LoadoutCheck_Glow_Inst
     tiling=NoTile
     imageSize=<944.0944 441.7221>
     margin=<ltrb 0 0 0 0>
     outlineSettings {
      bUseBrushTransparency=false
      roundingType=HalfHeightRadius
      width=0
      color {colorUseRule=UseColor_Specified, specifiedColor=<rgba 0 0 0 0>}
      cornerRadii=<0 0 0 0>
     }
     tintColor {colorUseRule=UseColor_Specified, specifiedColor=<rgba 1 1 1 1>}
     uVRegion {bIsValid=false, max=<0 0>, min=<0 0>}
    }
    colorAndOpacity=<rgba 0.5168 0.2131 1 1>
    renderTransform {angle=0, scale=<1.1 1.75>, shear=<0 0>, translation=<-450 0>}
   }
   slot {class=OverlaySlot, properties {verticalAlignment=VAlign_Center}}
  }
  Glow {
   widgetClassPath=&Image
   bIsVariable=true
   parent=Overlay_Background
   widget=Glow
   properties {
    visibility=SelfHitTestInvisible
    brush {
     bIsDynamicallyLoaded=false
     drawAs=Image
     imageType=NoImage
     mirroring=NoMirror
     resourceName=None
     resourceObject=/FortniteCoreUI/UI/Foundation/RockyRidge/Materials/M_UI_NightNight_Glow_Inst
     tiling=NoTile
     imageSize=<900 441.7221>
     margin=<ltrb 0 0 0 0>
     outlineSettings {
      bUseBrushTransparency=false
      roundingType=HalfHeightRadius
      width=0
      color {colorUseRule=UseColor_Specified, specifiedColor=<rgba 0 0 0 0>}
      cornerRadii=<0 0 0 0>
     }
     tintColor {colorUseRule=UseColor_Specified, specifiedColor=<rgba 1 1 1 1>}
     uVRegion {bIsValid=false, max=<0 0>, min=<0 0>}
    }
    colorAndOpacity=<rgba 0.5168 0.2131 1 1>
    renderTransform {angle=0, scale=<1.9 1.3>, shear=<0 0>, translation=<-348 -138>}
   }
   slot {class=OverlaySlot}
  }
  Glow_1 {
   widgetClassPath=&Image
   bIsVariable=true
   parent=Overlay_Background
   widget=Glow_1
   properties {
    renderOpacity=0.6762
    visibility=SelfHitTestInvisible
    brush {
     bIsDynamicallyLoaded=false
     drawAs=Image
     imageType=NoImage
     mirroring=NoMirror
     resourceName=None
     resourceObject=&M_UI_LoadoutCheck_Glow_Inst
     tiling=NoTile
     imageSize=<944.0944 441.7221>
     margin=<ltrb 0 0 0 0>
     outlineSettings {
      bUseBrushTransparency=false
      roundingType=HalfHeightRadius
      width=0
      color {colorUseRule=UseColor_Specified, specifiedColor=<rgba 0 0 0 0>}
      cornerRadii=<0 0 0 0>
     }
     tintColor {colorUseRule=UseColor_Specified, specifiedColor=<rgba 1 1 1 1>}
     uVRegion {bIsValid=false, max=<0 0>, min=<0 0>}
    }
    colorAndOpacity=<rgba 0.5168 0.2131 1 1>
    renderTransform {angle=0, scale=<2 1.45>, shear=<0 0>, translation=<-218 138>}
   }
   slot {class=OverlaySlot, properties {verticalAlignment=VAlign_Bottom}}
  }
  Glow_3 {
   widgetClassPath=&Image
   bIsVariable=true
   parent=Overlay_Background
   widget=Glow_3
   properties {
    visibility=SelfHitTestInvisible
    brush {
     bIsDynamicallyLoaded=false
     drawAs=Image
     imageType=NoImage
     mirroring=NoMirror
     resourceName=None
     resourceObject=&M_UI_LoadoutCheck_Glow_Inst
     tiling=NoTile
     imageSize=<944.0944 441.7221>
     margin=<ltrb 0 0 0 0>
     outlineSettings {
      bUseBrushTransparency=false
      roundingType=HalfHeightRadius
      width=0
      color {colorUseRule=UseColor_Specified, specifiedColor=<rgba 0 0 0 0>}
      cornerRadii=<0 0 0 0>
     }
     tintColor {colorUseRule=UseColor_Specified, specifiedColor=<rgba 1 1 1 1>}
     uVRegion {bIsValid=false, max=<0 0>, min=<0 0>}
    }
    colorAndOpacity=<rgba 0.5168 0.2131 1 1>
    renderTransform {angle=0, scale=<2.1 2.25>, shear=<0 0>, translation=<278 174>}
   }
   slot {class=OverlaySlot, properties {horizontalAlignment=HAlign_Right, verticalAlignment=VAlign_Bottom}}
  }
  Glow_2 {
   widgetClassPath=&Image
   bIsVariable=true
   parent=Overlay_Background
   widget=Glow_2
   properties {
    visibility=SelfHitTestInvisible
    brush {
     bIsDynamicallyLoaded=false
     drawAs=Image
     imageType=NoImage
     mirroring=NoMirror
     resourceName=None
     resourceObject=&M_UI_NightNight_Glow_Inst
     tiling=NoTile
     imageSize=<900 441.7221>
     margin=<ltrb 0 0 0 0>
     outlineSettings {
      bUseBrushTransparency=false
      roundingType=HalfHeightRadius
      width=0
      color {colorUseRule=UseColor_Specified, specifiedColor=<rgba 0 0 0 0>}
      cornerRadii=<0 0 0 0>
     }
     tintColor {colorUseRule=UseColor_Specified, specifiedColor=<rgba 1 1 1 1>}
     uVRegion {bIsValid=false, max=<0 0>, min=<0 0>}
    }
    colorAndOpacity=<rgba 0.5168 0.2131 1 1>
    renderTransform {angle=0, scale=<1.9 1.3>, shear=<0 0>, translation=<482 -138>}
   }
   slot {class=OverlaySlot, properties {horizontalAlignment=HAlign_Right}}
  }
  Image_1 {
   widgetClassPath=&Image
   bIsVariable=true
   parent=Overlay_Background
   widget=Image_1
   properties {
    renderOpacity=0.6476
    visibility=SelfHitTestInvisible
    brush {
     bIsDynamicallyLoaded=false
     drawAs=Image
     imageType=NoImage
     mirroring=NoMirror
     resourceName=None
     resourceObject=/FortniteCoreUI/UI/Foundation/RockyRidge/Materials/M_UI_NightNight_Ribbons_Inst_Inst
     tiling=NoTile
     imageSize=<944.0944 441.7221>
     margin=<ltrb 0 0 0 0>
     outlineSettings {
      bUseBrushTransparency=false
      roundingType=HalfHeightRadius
      width=0
      color {colorUseRule=UseColor_Specified, specifiedColor=<rgba 0 0 0 0>}
      cornerRadii=<0 0 0 0>
     }
     tintColor {colorUseRule=UseColor_Specified, specifiedColor=<rgba 1 1 1 1>}
     uVRegion {bIsValid=false, max=<0 0>, min=<0 0>}
    }
    renderTransform {angle=10, scale=<1 1.5>, shear=<0 0>, translation=<-64 0>}
   }
   slot {class=OverlaySlot, properties {verticalAlignment=VAlign_Center}}
  }
  Image_103 {
   widgetClassPath=&Image
   bIsVariable=true
   parent=Overlay_Background
   widget=Image_103
   properties {
    renderOpacity=0.7
    visibility=SelfHitTestInvisible
    brush {
     bIsDynamicallyLoaded=false
     drawAs=Image
     imageType=NoImage
     mirroring=NoMirror
     resourceName=None
     resourceObject=/FortniteCoreUI/UI/Foundation/RockyRidge/Materials/M_UI_NightNight_Ribbons_Inst
     tiling=NoTile
     imageSize=<944.0944 441.7221>
     margin=<ltrb 0 0 0 0>
     outlineSettings {
      bUseBrushTransparency=false
      roundingType=HalfHeightRadius
      width=0
      color {colorUseRule=UseColor_Specified, specifiedColor=<rgba 0 0 0 0>}
      cornerRadii=<0 0 0 0>
     }
     tintColor {colorUseRule=UseColor_Specified, specifiedColor=<rgba 1 1 1 1>}
     uVRegion {bIsValid=false, max=<0 0>, min=<0 0>}
    }
    renderTransform {angle=27, scale=<1 1>, shear=<0 0>, translation=<-64 0>}
   }
   slot {class=OverlaySlot}
  }
  Image_3 {
   widgetClassPath=&Image
   bIsVariable=true
   parent=Overlay_Background
   widget=Image_3
   properties {
    renderOpacity=0.6
    visibility=SelfHitTestInvisible
    brush {
     bIsDynamicallyLoaded=false
     drawAs=Image
     imageType=NoImage
     mirroring=NoMirror
     resourceName=None
     resourceObject=/FortniteCoreUI/UI/Foundation/RockyRidge/Materials/M_UI_NightNight_Ribbons_Inst2
     tiling=NoTile
     imageSize=<944.0944 441.7221>
     margin=<ltrb 0 0 0 0>
     outlineSettings {
      bUseBrushTransparency=false
      roundingType=HalfHeightRadius
      width=0
      color {colorUseRule=UseColor_Specified, specifiedColor=<rgba 0 0 0 0>}
      cornerRadii=<0 0 0 0>
     }
     tintColor {colorUseRule=UseColor_Specified, specifiedColor=<rgba 1 1 1 1>}
     uVRegion {bIsValid=false, max=<0 0>, min=<0 0>}
    }
    renderTransform {angle=-44, scale=<-1 -1>, shear=<0 0>, translation=<-80 -78>}
   }
   slot {class=OverlaySlot, properties {horizontalAlignment=HAlign_Right}}
  }
  Image {
   widgetClassPath=&Image
   bIsVariable=true
   parent=Overlay_Background
   widget=Image
   properties {
    renderOpacity=0.78
    visibility=SelfHitTestInvisible
    brush {
     bIsDynamicallyLoaded=false
     drawAs=Image
     imageType=NoImage
     mirroring=NoMirror
     resourceName=None
     resourceObject=&M_UI_NightNight_Ribbons_Inst
     tiling=NoTile
     imageSize=<944.0944 441.7221>
     margin=<ltrb 0 0 0 0>
     outlineSettings {
      bUseBrushTransparency=false
      roundingType=HalfHeightRadius
      width=0
      color {colorUseRule=UseColor_Specified, specifiedColor=<rgba 0 0 0 0>}
      cornerRadii=<0 0 0 0>
     }
     tintColor {colorUseRule=UseColor_Specified, specifiedColor=<rgba 1 1 1 1>}
     uVRegion {bIsValid=false, max=<0 0>, min=<0 0>}
    }
    renderTransform {angle=-27, scale=<-1 1>, shear=<0 0>, translation=<64 0>}
   }
   slot {class=OverlaySlot, properties {horizontalAlignment=HAlign_Right}}
  }
  Image_4 {
   widgetClassPath=&Image
   bIsVariable=true
   parent=Overlay_Background
   widget=Image_4
   properties {
    renderOpacity=0.6
    visibility=SelfHitTestInvisible
    brush {
     bIsDynamicallyLoaded=false
     drawAs=Image
     imageType=NoImage
     mirroring=NoMirror
     resourceName=None
     resourceObject=&M_UI_NightNight_Ribbons_Inst2
     tiling=NoTile
     imageSize=<944.0944 441.7221>
     margin=<ltrb 0 0 0 0>
     outlineSettings {
      bUseBrushTransparency=false
      roundingType=HalfHeightRadius
      width=0
      color {colorUseRule=UseColor_Specified, specifiedColor=<rgba 0 0 0 0>}
      cornerRadii=<0 0 0 0>
     }
     tintColor {colorUseRule=UseColor_Specified, specifiedColor=<rgba 1 1 1 1>}
     uVRegion {bIsValid=false, max=<0 0>, min=<0 0>}
    }
    renderTransform {angle=-10, scale=<1 -1>, shear=<0 0>, translation=<-131 0>}
   }
   slot {class=OverlaySlot, properties {verticalAlignment=VAlign_Bottom}}
  }
  Image_2 {
   widgetClassPath=&Image
   bIsVariable=true
   parent=Overlay_Background
   widget=Image_2
   properties {
    renderOpacity=0.6476
    visibility=SelfHitTestInvisible
    brush {
     bIsDynamicallyLoaded=false
     drawAs=Image
     imageType=NoImage
     mirroring=NoMirror
     resourceName=None
     resourceObject=&M_UI_NightNight_Ribbons_Inst_Inst
     tiling=NoTile
     imageSize=<944.0944 441.7221>
     margin=<ltrb 0 0 0 0>
     outlineSettings {
      bUseBrushTransparency=false
      roundingType=HalfHeightRadius
      width=0
      color {colorUseRule=UseColor_Specified, specifiedColor=<rgba 0 0 0 0>}
      cornerRadii=<0 0 0 0>
     }
     tintColor {colorUseRule=UseColor_Specified, specifiedColor=<rgba 1 1 1 1>}
     uVRegion {bIsValid=false, max=<0 0>, min=<0 0>}
    }
    renderTransform {angle=-29, scale=<1 -1.5>, shear=<0 0>, translation=<-54 55>}
   }
   slot {class=OverlaySlot, properties {verticalAlignment=VAlign_Bottom}}
  }
  Image_5 {
   widgetClassPath=&Image
   bIsVariable=true
   parent=Overlay_Background
   widget=Image_5
   properties {
    renderOpacity=0.6476
    visibility=SelfHitTestInvisible
    brush {
     bIsDynamicallyLoaded=false
     drawAs=Image
     imageType=NoImage
     mirroring=NoMirror
     resourceName=None
     resourceObject=&M_UI_NightNight_Ribbons_Inst_Inst
     tiling=NoTile
     imageSize=<944.0944 441.7221>
     margin=<ltrb 0 0 0 0>
     outlineSettings {
      bUseBrushTransparency=false
      roundingType=HalfHeightRadius
      width=0
      color {colorUseRule=UseColor_Specified, specifiedColor=<rgba 0 0 0 0>}
      cornerRadii=<0 0 0 0>
     }
     tintColor {colorUseRule=UseColor_Specified, specifiedColor=<rgba 1 1 1 1>}
     uVRegion {bIsValid=false, max=<0 0>, min=<0 0>}
    }
    renderTransform {angle=44, scale=<1 1.5>, shear=<0 0>, translation=<-87 -227>}
   }
   slot {class=OverlaySlot}
  }
  Image_6 {
   widgetClassPath=&Image
   bIsVariable=true
   parent=Overlay_Background
   widget=Image_6
   properties {
    renderOpacity=0.3048
    visibility=SelfHitTestInvisible
    brush {
     bIsDynamicallyLoaded=false
     drawAs=Image
     imageType=NoImage
     mirroring=NoMirror
     resourceName=None
     resourceObject=&M_UI_NightNight_Ribbons_Inst2
     tiling=NoTile
     imageSize=<944.0944 441.7221>
     margin=<ltrb 0 0 0 0>
     outlineSettings {
      bUseBrushTransparency=false
      roundingType=HalfHeightRadius
      width=0
      color {colorUseRule=UseColor_Specified, specifiedColor=<rgba 0 0 0 0>}
      cornerRadii=<0 0 0 0>
     }
     tintColor {colorUseRule=UseColor_Specified, specifiedColor=<rgba 1 1 1 1>}
     uVRegion {bIsValid=false, max=<0 0>, min=<0 0>}
    }
    renderTransform {angle=-156, scale=<1.6 1.8>, shear=<0 0>, translation=<-115 -11>}
   }
   slot {class=OverlaySlot, properties {horizontalAlignment=HAlign_Right, verticalAlignment=VAlign_Bottom}}
  }
  Image_7 {
   widgetClassPath=&Image
   bIsVariable=true
   parent=Overlay_Background
   widget=Image_7
   properties {
    renderOpacity=0.6476
    visibility=SelfHitTestInvisible
    brush {
     bIsDynamicallyLoaded=false
     drawAs=Image
     imageType=NoImage
     mirroring=NoMirror
     resourceName=None
     resourceObject=&M_UI_NightNight_Ribbons_Inst_Inst
     tiling=NoTile
     imageSize=<944.0944 441.7221>
     margin=<ltrb 0 0 0 0>
     outlineSettings {
      bUseBrushTransparency=false
      roundingType=HalfHeightRadius
      width=0
      color {colorUseRule=UseColor_Specified, specifiedColor=<rgba 0 0 0 0>}
      cornerRadii=<0 0 0 0>
     }
     tintColor {colorUseRule=UseColor_Specified, specifiedColor=<rgba 1 1 1 1>}
     uVRegion {bIsValid=false, max=<0 0>, min=<0 0>}
    }
    renderTransform {angle=-135, scale=<1 1.5>, shear=<0 0>, translation=<-198 262>}
   }
   slot {class=OverlaySlot, properties {horizontalAlignment=HAlign_Right, verticalAlignment=VAlign_Bottom}}
  }
  Image_8 {
   widgetClassPath=&Image
   bIsVariable=true
   parent=Overlay_Background
   widget=Image_8
   properties {
    renderOpacity=0.6476
    visibility=SelfHitTestInvisible
    brush {
     bIsDynamicallyLoaded=false
     drawAs=Image
     imageType=NoImage
     mirroring=NoMirror
     resourceName=None
     resourceObject=&M_UI_NightNight_Ribbons_Inst_Inst
     tiling=NoTile
     imageSize=<944.0944 441.7221>
     margin=<ltrb 0 0 0 0>
     outlineSettings {
      bUseBrushTransparency=false
      roundingType=HalfHeightRadius
      width=0
      color {colorUseRule=UseColor_Specified, specifiedColor=<rgba 0 0 0 0>}
      cornerRadii=<0 0 0 0>
     }
     tintColor {colorUseRule=UseColor_Specified, specifiedColor=<rgba 1 1 1 1>}
     uVRegion {bIsValid=false, max=<0 0>, min=<0 0>}
    }
    renderTransform {angle=0, scale=<-1 1.5>, shear=<0 0>, translation=<498 118>}
   }
   slot {class=OverlaySlot, properties {horizontalAlignment=HAlign_Right, verticalAlignment=VAlign_Center}}
  }
  ScaleBox_0 {
   widgetClassPath=&ScaleBox
   parent=Overlay_0
   widget=ScaleBox_0
   properties {stretchDirection=DownOnly}
   slot {class=OverlaySlot, properties {horizontalAlignment=HAlign_Fill, verticalAlignment=VAlign_Fill}}
  }
  SafeZone_0 {
   widgetClassPath=&SafeZone
   bIsVariable=true
   parent=ScaleBox_0
   widget=SafeZone_0
   slot {class=ScaleBoxSlot, properties {horizontalAlignment=HAlign_Fill, verticalAlignment=VAlign_Fill}}
  }
  CommonWidgetSwitcher_AnimStates {
   widgetClassPath=&CommonWidgetSwitcherLegacy
   bIsVariable=true
   parent=SafeZone_0
   widget=CommonWidgetSwitcher_AnimStates
   properties {activeWidgetIndex=3, transitionDuration=0.1, transitionType=Zoom}
   slot {class=SafeZoneSlot}
  }
  Spacer_57 {widgetClassPath=&Spacer, parent=CommonWidgetSwitcher_AnimStates, widget=Spacer_57, slot {class=WidgetSwitcherSlot}}
  SizeBox_Text {
   widgetClassPath=&SizeBox
   parent=CommonWidgetSwitcher_AnimStates
   widget=SizeBox_Text
   properties {bOverride_WidthOverride=true}
   slot {class=WidgetSwitcherSlot, properties {padding=<ltrb 50 0 50 0>}}
  }
  ScaleBox_4 {widgetClassPath=&ScaleBox, parent=SizeBox_Text, widget=ScaleBox_4, slot {class=SizeBoxSlot}}
  Overlay_TextLines {widgetClassPath=&Overlay, bIsVariable=true, parent=ScaleBox_4, widget=Overlay_TextLines, slot {class=ScaleBoxSlot}}
  Text_line1 {
   widgetClassPath=&CommonTextBlock
   bIsVariable=true
   parent=Overlay_TextLines
   widget=Text_line1
   properties {
    justification=Center
    text=To be
    textTransformPolicy=ToUpper
    font {
     bForceMonospaced=false
     bMaterialIsStencil=false
     fontMaterial=null
     fontObject=&Burbank
     letterSpacing=0
     monospacedWidth=1
     size=300
     skewAmount=0
     typefaceFontName=BurbankRegular-Black
     outlineSettings {
      bApplyOutlineToDropShadows=false
      bMiteredCorners=false
      bSeparateFillAlpha=false
      outlineMaterial=null
      outlineSize=0
      outlineColor=<rgba 0.7305 0.016 0.0152 1>
     }
    }
    margin=<ltrb 0 -100 0 -100>
    renderTransform {angle=0, scale=<0.25 0.25>, shear=<-8 0>, translation=<-162.2781 -355.9444>}
    shadowOffset=<0 0>
   }
   slot {class=OverlaySlot, properties {horizontalAlignment=HAlign_Center}}
  }
  Text_line2 {
   widgetClassPath=&CommonTextBlock
   bIsVariable=true
   parent=Overlay_TextLines
   widget=Text_line2
   properties {
    justification=Center
    text=Continued...
    textTransformPolicy=ToUpper
    font {
     bForceMonospaced=false
     bMaterialIsStencil=false
     fontMaterial=null
     fontObject=&Burbank
     letterSpacing=0
     monospacedWidth=1
     size=300
     skewAmount=0
     typefaceFontName=BurbankRegular-Black
     outlineSettings {
      bApplyOutlineToDropShadows=false
      bMiteredCorners=false
      bSeparateFillAlpha=false
      outlineMaterial=null
      outlineSize=0
      outlineColor=<rgba 0.7305 0.016 0.0152 1>
     }
    }
    margin=<ltrb 0 -100 0 -100>
    renderTransform {angle=0, scale=<0.5 0.5>, shear=<-8 0>, translation=<-18.5877 -150.5964>}
    renderTransformPivot=<0.5 0>
    shadowOffset=<0 0>
   }
   slot {class=OverlaySlot, properties {horizontalAlignment=HAlign_Center}}
  }
  Text_line3 {
   widgetClassPath=&CommonTextBlock
   bIsVariable=true
   parent=Overlay_TextLines
   widget=Text_line3
   properties {
    justification=Center
    textTransformPolicy=ToUpper
    font {
     bForceMonospaced=false
     bMaterialIsStencil=false
     fontMaterial=null
     fontObject=&Burbank
     letterSpacing=0
     monospacedWidth=1
     size=300
     skewAmount=0
     typefaceFontName=BurbankRegular-Black
     outlineSettings {
      bApplyOutlineToDropShadows=false
      bMiteredCorners=false
      bSeparateFillAlpha=false
      outlineMaterial=null
      outlineSize=0
      outlineColor=<rgba 0.7305 0.016 0.0152 1>
     }
    }
    margin=<ltrb 0 -100 0 -100>
    renderTransform {angle=0, scale=<1 1>, shear=<-8 0>, translation=<0 0>}
    renderTransformPivot=<0.5 0>
    shadowOffset=<0 0>
   }
   slot {class=OverlaySlot, properties {horizontalAlignment=HAlign_Center}}
  }
  Text_line4 {
   widgetClassPath=&CommonTextBlock
   bIsVariable=true
   parent=Overlay_TextLines
   widget=Text_line4
   properties {
    justification=Center
    text=Line4
    textTransformPolicy=ToUpper
    font {
     bForceMonospaced=false
     bMaterialIsStencil=false
     fontMaterial=null
     fontObject=&Burbank
     letterSpacing=0
     monospacedWidth=1
     size=300
     skewAmount=0
     typefaceFontName=BurbankRegular-Black
     outlineSettings {
      bApplyOutlineToDropShadows=false
      bMiteredCorners=false
      bSeparateFillAlpha=false
      outlineMaterial=null
      outlineSize=6
      outlineColor=<rgba 0.7305 0.016 0.0152 1>
     }
    }
    margin=<ltrb 0 -100 0 -100>
    renderTransform {angle=0, scale=<1 1>, shear=<-8 0>, translation=<0 0>}
    renderTransformPivot=<0.5 0>
    shadowOffset=<0 0>
   }
   slot {class=OverlaySlot, properties {horizontalAlignment=HAlign_Center}}
  }
  Overlay_SeasonTitle {widgetClassPath=&Overlay, parent=CommonWidgetSwitcher_AnimStates, widget=Overlay_SeasonTitle, slot {class=WidgetSwitcherSlot}}
  RetainerBox_0 {
   widgetClassPath=/Script/UMG.RetainerBox
   bIsVariable=true
   parent=Overlay_SeasonTitle
   widget=RetainerBox_0
   properties {
    effectMaterial=/FortniteCoreUI/UI/Foundation/RockyRidge/Materials/M_UI_TextMask
    renderOnInvalidation=true
    renderOnPhase=false
    visibility=SelfHitTestInvisible
   }
   slot {class=OverlaySlot, properties {horizontalAlignment=HAlign_Fill, verticalAlignment=VAlign_Fill}}
  }
  Overlay_1 {widgetClassPath=&Overlay, parent=RetainerBox_0, widget=Overlay_1, slot {class=PanelSlot}}
  VerticalBox {
   widgetClassPath=&VerticalBox
   parent=Overlay_1
   widget=VerticalBox
   slot {class=OverlaySlot, properties {horizontalAlignment=HAlign_Center, verticalAlignment=VAlign_Center}}
  }
  Text_InSeason5_Line1 {
   widgetClassPath=&CommonTextBlock
   parent=VerticalBox
   widget=Text_InSeason5_Line1
   properties {
    justification=Center
    text=In Season
    textTransformPolicy=ToUpper
    font {
     bForceMonospaced=false
     bMaterialIsStencil=false
     fontMaterial=null
     fontObject=&Burbank
     letterSpacing=0
     monospacedWidth=1
     size=80
     skewAmount=0
     typefaceFontName=BurbankRegular-Black
     outlineSettings {
      bApplyOutlineToDropShadows=false
      bMiteredCorners=false
      bSeparateFillAlpha=false
      outlineMaterial=null
      outlineSize=0
      outlineColor=<rgba 0 0 0 1>
     }
    }
    renderTransform {angle=0, scale=<1 1>, shear=<-12 0>, translation=<0 0>}
    shadowOffset=<0 0>
   }
   slot {class=VerticalBoxSlot, properties {horizontalAlignment=HAlign_Center}}
  }
  ScaleBox_Seasonnumber {
   widgetClassPath=&ScaleBox
   parent=VerticalBox
   widget=ScaleBox_Seasonnumber
   slot {class=VerticalBoxSlot, properties {padding=<ltrb 0 -170 0 -80>}}
  }
  Text_InSeason5_Line2 {
   widgetClassPath=&CommonTextBlock
   parent=ScaleBox_Seasonnumber
   widget=Text_InSeason5_Line2
   properties {
    justification=Center
    text="Five!"
    textTransformPolicy=ToUpper
    font {
     bForceMonospaced=false
     bMaterialIsStencil=false
     fontMaterial=null
     fontObject=&Burbank
     letterSpacing=0
     monospacedWidth=1
     size=500
     skewAmount=0
     typefaceFontName=BurbankRegular-Black
     outlineSettings {
      bApplyOutlineToDropShadows=false
      bMiteredCorners=false
      bSeparateFillAlpha=false
      outlineMaterial=null
      outlineSize=0
      outlineColor=<rgba 0 0 0 1>
     }
    }
    renderTransform {angle=0, scale=<1 1>, shear=<-12 0>, translation=<0 0>}
    shadowOffset=<0 0>
   }
   slot {class=ScaleBoxSlot}
  }
  Overlay_CountdownSlide {widgetClassPath=&Overlay, parent=CommonWidgetSwitcher_AnimStates, widget=Overlay_CountdownSlide, slot {class=WidgetSwitcherSlot}}
  VerticalBox_0 {
   widgetClassPath=&VerticalBox
   parent=Overlay_CountdownSlide
   widget=VerticalBox_0
   slot {class=OverlaySlot, properties {horizontalAlignment=HAlign_Center, verticalAlignment=VAlign_Center, padding=<ltrb 64 0 64 0>}}
  }
  Image_Logo {
   widgetClassPath=&Image
   bIsVariable=true
   parent=VerticalBox_0
   widget=Image_Logo
   properties {
    visibility=SelfHitTestInvisible
    brush {
     bIsDynamicallyLoaded=false
     drawAs=Image
     imageType=NoImage
     mirroring=NoMirror
     resourceName=None
     resourceObject=/FortniteCoreUI/Foundation/Textures/Logo/T-Logo-Wordmark-1024x256
     tiling=NoTile
     imageSize=<700 170>
     margin=<ltrb 0 0 0 0>
     outlineSettings {
      bUseBrushTransparency=false
      roundingType=HalfHeightRadius
      width=0
      color {colorUseRule=UseColor_Specified, specifiedColor=<rgba 0 0 0 0>}
      cornerRadii=<0 0 0 0>
     }
     tintColor {colorUseRule=UseColor_Specified, specifiedColor=<rgba 1 1 1 1>}
     uVRegion {bIsValid=false, max=<0 0>, min=<0 0>}
    }
   }
   slot {class=VerticalBoxSlot, properties {horizontalAlignment=HAlign_Center, verticalAlignment=VAlign_Center, padding=<ltrb 0 0 0 50>}}
  }
  CommonBorder_0 {
   widgetClassPath=&CommonBorder
   parent=VerticalBox_0
   widget=CommonBorder_0
   properties {
    horizontalAlignment=HAlign_Center
    verticalAlignment=VAlign_Center
    padding=<ltrb 0 0 0 0>
    renderTransform {angle=0, scale=<1 1>, shear=<-12 0>, translation=<0 0>}
   }
   slot {class=VerticalBoxSlot, properties {padding=<ltrb 0 -120 0 0>}}
  }
  Text_Season5_Line2 {
   widgetClassPath=&CommonTextBlock
   parent=CommonBorder_0
   widget=Text_Season5_Line2
   properties {
    justification=Center
    text=Season 5
    textTransformPolicy=ToUpper
    font {
     bForceMonospaced=false
     bMaterialIsStencil=false
     fontMaterial=/FortniteCoreUI/UI/Foundation/RockyRidge/Materials/M_UI_NightNight_Text
     fontObject=&Burbank
     letterSpacing=0
     monospacedWidth=1
     size=240
     skewAmount=0
     typefaceFontName=BurbankRegular-Black
     outlineSettings {
      bApplyOutlineToDropShadows=false
      bMiteredCorners=false
      bSeparateFillAlpha=false
      outlineMaterial=/FortniteCoreUI/UI/Foundation/RockyRidge/Materials/M_UI_NightNight_Text_Outline
      outlineSize=4
      outlineColor=<rgba 1 1 1 1>
     }
    }
    renderTransformPivot=<0.4429 0.2093>
    shadowOffset=<0 0>
   }
   slot {class=BorderSlot, properties {horizontalAlignment=HAlign_Center, verticalAlignment=VAlign_Center, padding=<ltrb 0 0 0 0>}}
  }
  VerticalBox_2 {
   widgetClassPath=&VerticalBox
   parent=Overlay_CountdownSlide
   widget=VerticalBox_2
   slot {class=OverlaySlot, properties {horizontalAlignment=HAlign_Right, verticalAlignment=VAlign_Bottom, padding=<ltrb 50 50 50 50>}}
  }
  Text_Countdown_StartsIn {
   widgetClassPath=&CommonTextBlock
   parent=VerticalBox_2
   widget=Text_Countdown_StartsIn
   properties {
    justification=Center
    text=Starts in...
    textTransformPolicy=ToUpper
    font {
     bForceMonospaced=false
     bMaterialIsStencil=false
     fontMaterial=null
     fontObject=&Burbank
     letterSpacing=0
     monospacedWidth=1
     size=45
     skewAmount=0
     typefaceFontName=BurbankRegular-Black
     outlineSettings {
      bApplyOutlineToDropShadows=false
      bMiteredCorners=false
      bSeparateFillAlpha=false
      outlineMaterial=null
      outlineSize=0
      outlineColor=<rgba 0 0 0 1>
     }
    }
    shadowOffset=<0 0>
   }
   slot {class=VerticalBoxSlot, properties {padding=<ltrb 0 0 0 -30>}}
  }
  SizeBox_1 {
   widgetClassPath=&SizeBox
   parent=VerticalBox_2
   widget=SizeBox_1
   properties {widthOverride=546.966}
   slot {class=VerticalBoxSlot}
  }
  Text_Countdown_value {
   widgetClassPath=&CommonTextBlock
   bIsVariable=true
   parent=SizeBox_1
   widget=Text_Countdown_value
   properties {
    minDesiredWidth=580
    text=00:00:00:00
    textTransformPolicy=ToUpper
    colorAndOpacity {colorUseRule=UseColor_Specified, specifiedColor=<rgba 0.1717 0.6806 0.8308 1>}
    font {
     bForceMonospaced=false
     bMaterialIsStencil=false
     fontMaterial=null
     fontObject=&Burbank
     letterSpacing=0
     monospacedWidth=1
     size=72
     skewAmount=0
     typefaceFontName=BurbankRegular-Black
     outlineSettings {
      bApplyOutlineToDropShadows=false
      bMiteredCorners=false
      bSeparateFillAlpha=false
      outlineMaterial=null
      outlineSize=0
      outlineColor=<rgba 0 0 0 1>
     }
    }
    shadowOffset=<0 0>
   }
   slot {class=SizeBoxSlot, properties {horizontalAlignment=HAlign_Center, verticalAlignment=VAlign_Center}}
  }
  CommonVisibilityWidget_114 {
   widgetClassPath=&CommonVisibilityWidgetLegacy
   parent=Overlay_CountdownSlide
   widget=CommonVisibilityWidget_114
   properties {
    bShowForAndroid=false
    bShowForErebus=false
    bShowForIOS=false
    bShowForPS4=false
    bShowForPS5=false
    bShowForSwitch2=false
    bShowForXBox=false
    bShowForXSX=false
    horizontalAlignment=HAlign_Left
    verticalAlignment=VAlign_Bottom
    padding=<ltrb 50 50 50 50>
   }
   slot {class=OverlaySlot, properties {verticalAlignment=VAlign_Bottom}}
  }
  CloseButton {
   widgetClassPath=/FortUILibrary/Foundation/Buttons/SkewButton
   bIsVariable=true
   parent=CommonVisibilityWidget_114
   widget=CloseButton
   properties {
    bIsFocusable=true
    onButtonBaseClicked=(&self:WidgetTree.CloseButton.HandleOnButtonClicked)
    onButtonBaseDoubleClicked=(&self:WidgetTree.CloseButton.HandleOnButtonDoubleClicked)
    onButtonBaseHovered=(&self:WidgetTree.CloseButton.HandleOnButtonHovered)
    onButtonBaseUnhovered=(&self:WidgetTree.CloseButton.HandleOnButtonUnhovered)
    onSelectedChangedBase=(&self:WidgetTree.CloseButton.HandleOnSelectedChanged)
    buttonColor=<rgba 1 1 1 1>
    hoverColor=<rgba 1 0.0125 0 1>
    triggeringInputAction {dataTable=&InputActionDataTable, rowName=GenericFaceButtonTop}
   }
   slot {class=BorderSlot, properties {horizontalAlignment=HAlign_Left, verticalAlignment=VAlign_Bottom, padding=<ltrb 50 50 50 50>}}
  }
  Text_Button_Close {
   widgetClassPath=&CommonTextBlock
   bIsVariable=true
   namedSlotHost=CloseButton
   widget=Text_Button_Close
   properties {
    text=Exit
    textTransformPolicy=ToUpper
    colorAndOpacity {colorUseRule=UseColor_Specified, specifiedColor=<rgba 0 0 0 1>}
    font {
     bForceMonospaced=false
     bMaterialIsStencil=false
     fontMaterial=null
     fontObject=&Burbank
     letterSpacing=0
     monospacedWidth=1
     size=24
     skewAmount=0
     typefaceFontName=BurbankRegular-Black
     outlineSettings {
      bApplyOutlineToDropShadows=false
      bMiteredCorners=false
      bSeparateFillAlpha=false
      outlineMaterial=null
      outlineSize=0
      outlineColor=<rgba 0 0 0 1>
     }
    }
    renderTransform {angle=0, scale=<1 1>, shear=<-8 0>, translation=<0 0>}
    shadowOffset=<0 0>
   }
  }
  Image_Fade {
   widgetClassPath=&Image
   bIsVariable=true
   parent=Overlay_0
   widget=Image_Fade
   properties {renderOpacity=0, visibility=HitTestInvisible, colorAndOpacity=<rgba 0 0 0 1>}
   slot {class=OverlaySlot, properties {horizontalAlignment=HAlign_Fill, verticalAlignment=VAlign_Fill}}
  }
  Text_CountDownTimer {widgetClassPath=&CommonTextBlock, bInherited=true, widget=null}
  Button_LeaveExperience {widgetClassPath=&CommonButtonLegacy, bInherited=true, widget=null}
  ContextualHelpText {widgetClassPath=&CommonTextBlock, bInherited=true, widget=null}
  WidgetStack_PrimaryContent {widgetClassPath=/Script/CommonUI.CommonActivatableWidgetStack, bInherited=true, widget=null}
 ]
}
